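# UI material for the ApexWild item shop store background, building a dark, patterned gradient backdrop.
# A linear gradient with a panner-driven rotation angle blends with a pale blue tint through a radial gradient.
# A scaled, rotated T_UI_Mesh texture masks that color over near-black, and a vertical falloff darkens the top and bottom edges.
# A squared time-sine chain and a purple constant feed nothing.
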Material /ItemShop/UI/Art/ApexWild/StoreBG/M_UI_ApexWild_StoreBG
graph {
 properties {bUsedWithStaticMesh=false, materialDomain=MD_UI, pixelDepthOffsetMode=PDOM_Legacy, usedShadingModels=Default Lit}
 nodes [
  Constant3Vector() {constant=<rgba 0.1369 0.0894 0.3594 0>}
  n1 = Constant3Vector() {constant=<rgba 0.6055 0.6416 0.9688 0>}
  n2 = ScalarParameter() {defaultValue=0.336, parameterName=AngleRotation}
  n3 = ScalarParameter() {defaultValue=0.05, parameterName=Param_2}
  n4 = LinearInterpolate(
   A=Constant3Vector() {constant=<rgba 0.0146 0.0151 0.026 0>}
   B=Constant3Vector() {constant=<rgba 0.1683 0.2148 0.2448 0>}
   Alpha=MaterialFunctionCall(
    GradientStart=ScalarParameter() {defaultValue=0.0976}
    GradientEnd=ScalarParameter() {defaultValue=3.3059, parameterName=Param_1}
    RotationAngle=Panner(Coordinate=@n2, Speed=@n3)
   ) {materialFunction=&MF_UI_LinearGradient}.Result
  )
  n5 = MaterialFunctionCall(
   UV=Saturate(None=TextureCoordinate() {uTiling=1.3434, vTiling=1.3138})
   GradientStart=ScalarParameter() {defaultValue=3.6968, parameterName=RadialGradientstart}
   GradientEnd=ScalarParameter() {defaultValue=0.0621, parameterName=RadialGradientend}
  ) {materialFunction=&MF_UI_RadialGradient}
  n6 = MaterialFunctionCall(
   UVs=Multiply(
    A=MaterialFunctionCall() {materialFunction=&MF_UI_UVScaling}.Scaled to Fit
    B=ScalarParameter() {defaultValue=2.9286, parameterName=TextureSize}
   )
   Rotation Center=Constant2Vector() {g=50, r=50}
   "Rotation Angle (0-1)"=ScalarParameter() {defaultValue=-0.136, parameterName=TextureRotation}
  ) {materialFunction=&CustomRotator}
  n7 = Abs(None=ConstantBiasScale(None=TextureCoordinate()) {bias=-0.5, scale=2})
  n8 = Multiply(
   A=Subtract(A=ComponentMask(None=@n7) {g=true}, B=ScalarParameter() {defaultValue=0.9246, parameterName=Subtract})
   B=ScalarParameter() {defaultValue=-1.3011, parameterName=Falloff}
  )
  LinearInterpolate(
   A=Constant3Vector() {constant=<rgba 0.0088 0.0091 0.0156 0>}
   B=LinearInterpolate(
    A=Constant3Vector() {constant=<rgba 0.0146 0.0151 0.026 0>}
    B=LinearInterpolate(A=@n1, B=@n4, Alpha=@n5.Smooth)
    Alpha=TextureSample(UVs=@n6.Rotated Values) {texture=&/T_UI_Mesh}.R
   )
   Alpha=Multiply(A=Saturate(None=@n8), B=ScalarParameter() {defaultValue=1.7725, parameterName=Intensity})
  )
  n9 = Multiply(A=Time(), B=ScalarParameter() {defaultValue=0.006, parameterName=tenminutes})
  Multiply(A=Sine(None=@n9), B=Sine(None=@n9))
 ]
}
